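# HUD widget that shows in-world damage numbers, built on the native FortUserWidget_DamageNumbers class.
# An inherited overlay-type DynamicEntryBox of WBP_DamageNumber entries sits inside a HUD element wrapper tagged to hide for capture and post-game.
# On Construct it listens for spectating-target changes and HUD-context pawn changes, then calls OnPawnSet at once.
# A valid spectated pawn becomes BoundPawn between ClearBinding and UpdateBinding, so the numbers follow the spectated player.
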WidgetBlueprint /DamageNumbers/DamageNumbersWidget
graphs [EventGraph {nodes [
   n1 = On Spectating Target Changed!() -> (OutputDelegate, then, SpectatingTarget:Fort Player State Zone Object Reference) {is=title K2Node_CustomEvent}
   > Is Valid(InputObject=@n1.SpectatingTarget)
   n2 = GetPawn!(self:Player State Object Reference=@n1.SpectatingTarget) -> (ReturnValue:Pawn Object Reference)
   > Is Valid(InputObject=@n2)
   > n3 = Cast To FortPlayerPawn(Object=@n2)
   > ClearBinding(self:Fort User Widget Damage Numbers Object Reference)
   > Set BoundPawn(BoundPawn:Fort Player Pawn Object Reference=@n3) -> (Output_Get:Fort Player Pawn Object Reference)
   > UpdateBinding(self:Fort User Widget Damage Numbers Object Reference)
   Event Construct()
   > n4 = Sequence()
   > n5 = Cast To FortPlayerController(Object=GetOwningPlayer())
   > n6 = Cast To FortPlayerStateAthena!(execute, Object=Get PlayerState!(self:Controller Object Reference=@n5) -> (PlayerState:Player State Object Reference)) -> (then, CastFailed, AsFort Player State Athena:Fort Player State Athena Object Reference) {is=title K2Node_DynamicCast}
   > Bind Event to On Spectating Target Changed Event(self:Fort Player State Zone Object Reference=@n6, Delegate=Create Event()) {is=title K2Node_AddDelegate}
   Bind Event to On Pawn Set(
    execute=@n4.then_1
    self:Fort HUDContext Object Reference=Get FortHUDContext!() -> (ReturnValue:Fort HUDContext Object Reference) {is=title K2Node_GetBlueprintContextEx}
    Delegate=Create Event()
   ) {is=title K2Node_AddDelegate}
   > OnPawnSet(self:Fort User Widget Damage Numbers Object Reference)
  ]}]
variables [
 CumulativeStyle {type=boolean, default=true}
 DidBind {type=boolean}
]
widgets {
 info {
  inheritedWidgetCount=1
  namedSlotCount=1
  parentClass=/Script/DamageNumbersUI.FortUserWidget_DamageNumbers
  rootWidgetClass=/Game/Athena/HUD/FortHUDElementWrapper.FortHUDElementWrapper_C
  widgetCount=3
 }
 widgets [
  FortHUDElementWrapper {
   widgetClassPath=/Game/Athena/HUD/FortHUDElementWrapper
   bIsVariable=true
   widget=FortHUDElementWrapper
   properties {
    elementTag {tagName=HUD.InWorld.DamageNumbers}
    hUDElementTag {gameplayTags [{tagName=Athena.HideForCapture}, {tagName=HUD.Athena.HideForPostGame}]}
   }
  }
  Overlay_0 {widgetClassPath=&Overlay, namedSlotHost=FortHUDElementWrapper, widget=Overlay_0}
  DynamicEntryBox_Numbers {
   widgetClassPath=&DynamicEntryBox
   bInherited=true
   bIsVariable=true
   parent=Overlay_0
   widget=DynamicEntryBox_Numbers
   properties {entryBoxType=Overlay, entryWidgetClass=&/WBP_DamageNumber.WBP_DamageNumber_C}
   slot {class=OverlaySlot}
  }
 ]
}
