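# Translucent UI material for the inner stroke of damage numbers, judging by its name and parameter.
# The graph holds a single InnerStrokeColor vector parameter, defaulting to pink, which exposes the stroke color to material instances.
# It has no textures, math or other nodes, and no link from the parameter to a material output is shown.
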Material /DamageNumbers/Materials/M_UI_DamageNumbersInnerStroke
graph {
 properties {blendMode=BLEND_Translucent, materialDomain=MD_UI, pixelDepthOffsetMode=PDOM_Legacy, usedShadingModels=Default Lit}
 nodes [VectorParameter() {parameterName=InnerStrokeColor, defaultValue=<rgba 1 0.3098 0.7725 1>}]
}
