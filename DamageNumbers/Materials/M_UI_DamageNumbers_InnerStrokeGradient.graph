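# UI-domain material that produces a two-color linear gradient, named for the inner stroke of damage number text.
# A lerp blends the Color01 and Color02 parameters, black and white by default, using MF_UI_LinearGradient as the alpha.
# The gradient function takes texture coordinates plus GradientStart and RotationAngle scalar parameters, exposing the gradient's start and direction as parameters.
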Material /DamageNumbers/Materials/M_UI_DamageNumbers_InnerStrokeGradient
graph {
 properties {bAutomaticallySetUsageInEditor=false, materialDomain=MD_UI, usedShadingModels=Default Lit}
 nodes [LinearInterpolate(
   A=VectorParameter() {parameterName=Color01, defaultValue=<rgba 0 0 0 1>}.RGB
   B=VectorParameter() {parameterName=Color02, defaultValue=<rgba 1 1 1 1>}.RGB
   Alpha=MaterialFunctionCall(
    UVs=TextureCoordinate()
    GradientStart=ScalarParameter() {parameterName=GradientStart}
    RotationAngle=ScalarParameter() {defaultValue=0.125, parameterName=RotationAngle}
   ) {materialFunction=&MF_UI_LinearGradient}.Result
  )]
}
