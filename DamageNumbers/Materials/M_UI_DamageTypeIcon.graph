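# UI material that draws the damage-type icon (vehicle, shield or animal) or crit icon for damage numbers from a channel-packed distance-field texture.
# IconShape blends between texture channels, and ShowCrit blends that result toward the crit shape.
# Distance-field functions produce a tinted fill and a vertical-gradient outline, whose summed masks form a combined shape.
# Parameter-driven UV offsets and center scaling define separate type and crit icon placement.
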Material /DamageNumbers/Materials/M_UI_DamageTypeIcon
graph {
 properties {blendMode=BLEND_Translucent, materialDomain=MD_UI, pixelDepthOffsetMode=PDOM_Legacy, usedShadingModels=Default Lit}
 nodes [
  n1 = TextureSample() {samplerSource=SSM_Clamp_WorldGroupSettings, samplerType=SAMPLERTYPE_LinearColor, texture=&/T_UI_DamageNumbersTypes_DF}
  n2 = ScalarParameter() {defaultValue=1, desc="0-vehicle, 0.5-shield, 1-animal", parameterName=IconShape}
  n3 = Clamp(None=Multiply(A=@n2) {constB=2})
  n4 = Clamp(None=Multiply(A=Add(A=@n2) {constB=-0.5}) {constB=2})
  n5 = LinearInterpolate(
   A=LinearInterpolate(A=LinearInterpolate(A=@n1.R, B=@n1.R, Alpha=@n3), B=LinearInterpolate(A=@n1.G, B=@n1.G, Alpha=@n4), Alpha=@n2) {desc="Blend it!"}
   B=NamedRerouteUsage()
   Alpha=ScalarParameter() {parameterName=ShowCrit}
  ) {constAlpha=0}
  n6 = MaterialFunctionCall!(RGBA Channels=@n5, Glow Min=Constant() {r=0.8}, Glow Max=Constant() {r=0.9}, Outline Thickness) {materialFunction=/FortUILibrary/MaterialFunctions/Basics/MF_UI_ChannelPackedDistanceField}
  n7 = MaterialFunctionCall(RGBA Channels=@n5, Glow Min=Constant() {r=0.0907}, Glow Max=Constant() {r=0.3394}, Outline Thickness=Constant() {r=0.4}) {materialFunction=&MF_UI_ChannelPackedDistanceField}
  n8 = LinearInterpolate(
   A=NamedRerouteUsage()
   B=NamedRerouteUsage()
   Alpha=MaterialFunctionCall() {materialFunction=&LinearGradient}.VGradient
  )
  Add(A=Multiply(A=@n6.Fill, B=NamedRerouteUsage()), B=Multiply(A=@n7.Outline, B=Power(Base=@n8)))
  n9 = TextureCoordinate()
  n10 = Add(
   A=ComponentMask(None=@n9) {r=true}
   B=Multiply(A=ScalarParameter() {defaultValue=-0.2, parameterName=CritPosition}, B=ScalarParameter() {defaultValue=1, parameterName=ShowType})
  ) {constB=-0.2}
  n11 = ComponentMask(None=@n9) {g=true}
  NamedRerouteDeclaration(None=MaterialFunctionCall(UVs=AppendVector(A=@n10, B=@n11), Texture Scale=ScalarParameter() {defaultValue=0.8, parameterName=CritIconScale}) {materialFunction=&ScaleUVsByCenter}.UVs) {name=CritIconUV's, nodeColor=<rgba 0 0 0 1>}
  NamedRerouteDeclaration(None=@n1.A) {name=CritIconShape, nodeColor=<rgba 0 0 0 1>}
  NamedRerouteDeclaration(None=VectorParameter() {parameterName=TypeMainColor, defaultValue=<rgba 1 1 1 1>}.RGB) {name=Main, nodeColor=<rgba 0 0 0 1>}
  NamedRerouteDeclaration(None=VectorParameter() {parameterName=TypeOutline1, defaultValue=<rgba 0 0 0 1>}.RGB) {name=Outline1, nodeColor=<rgba 0 0 0 1>}
  NamedRerouteDeclaration(None=VectorParameter() {parameterName=TypeOutline2, defaultValue=<rgba 0 0 0 1>}.RGB) {name=Outline2, nodeColor=<rgba 0 0 0 1>}
  n12 = TextureCoordinate()
  n13 = Add(
   A=ComponentMask(None=@n12) {r=true}
   B=Multiply(A=ScalarParameter() {defaultValue=0.2, parameterName=TypePosition}, B=ScalarParameter() {parameterName=ShowCrit})
  ) {constB=0.2}
  n14 = ComponentMask(None=@n12) {g=true}
  NamedRerouteDeclaration(None=MaterialFunctionCall(UVs=AppendVector(A=@n13, B=@n14), Texture Scale=ScalarParameter() {defaultValue=0.9, parameterName=TypeIconScale}) {materialFunction=&ScaleUVsByCenter}.UVs) {name=TypeIconUV's, nodeColor=<rgba 0 0 0 1>}
  Reroute(None=Add(A=@n6.Fill, B=@n7.Outline))
 ]
}
