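# UI material for an EMP grenade shield icon in the DamageNumbers content, with intact and broken states.
# One texture supplies the shield mask in red and the broken-shield mask in green through named reroutes.
# Animate parameters blend start and stop colors and alphas, including a bolt alpha for the intact shield.
# IsBroken blends between the two states, and Final_Alpha scales the saturated opacity of the translucent result.
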Material /DamageNumbers/Materials/M_UI_EmpGrenadeShieldIcon
graph {
 properties {blendMode=BLEND_Translucent, materialDomain=MD_UI, pixelDepthOffsetMode=PDOM_Legacy, usedShadingModels=Default Lit}
 nodes [
  LinearInterpolate(
   A=LinearInterpolate(
    A=VectorParameter() {group=Color, parameterName=Shield_Color, defaultValue=<rgba 0.0319 0.8388 1 1>}.RGB
    B=VectorParameter() {group=Color, parameterName=Shield_Color_Stop, defaultValue=<rgba 0.0319 0.8388 1 1>}.RGB
    Alpha=ScalarParameter() {parameterName=AnimateShieldColor}
   )
   B=LinearInterpolate(
    A=VectorParameter() {group=Color, parameterName=BrokenShield_Color_Start, defaultValue=<rgba 0.0319 0.8388 1 1>}.RGB
    B=VectorParameter() {group=Color, parameterName=BrokenShield_Color_Stop, defaultValue=<rgba 0.0319 0.8388 1 1>}.RGB
    Alpha=ScalarParameter() {parameterName=AnimateBrokenColor}
   )
   Alpha=ScalarParameter() {parameterName=IsBroken}
  )
  n1 = TextureSample() {samplerType=SAMPLERTYPE_LinearColor, texture=&/T_UI_EMP_GrenadeShield}
  NamedRerouteDeclaration(None=@n1.R) {name=Shield, nodeColor=<rgba 0 0 0 1>}
  NamedRerouteDeclaration(None=@n1.G) {name=ShieldBroken, nodeColor=<rgba 0 0 0 1>}
  n2 = ScalarParameter() {defaultValue=1, group=Alpha, parameterName=Shield_Alpha}
  n3 = LinearInterpolate(
   A=ScalarParameter() {defaultValue=1, group=Alpha, parameterName=Bolt_Alpha}
   B=ScalarParameter() {defaultValue=1, group=Alpha, parameterName=Bolt_Alpha_Stop}
   Alpha=ScalarParameter() {parameterName=AnimateBoltAlpha}
  )
  n4 = Saturate(None=DotProduct(A=NamedRerouteUsage(), B=LinearInterpolate(A=@n2, B=@n3)))
  n5 = ScalarParameter() {defaultValue=1, group=Alpha, parameterName=BrokenShield_Alpha_Start}
  n6 = ScalarParameter() {defaultValue=1, group=Alpha, parameterName=BrokenShield_Alpha_Stop}
  n7 = ScalarParameter() {parameterName=AnimateBrokenAlpha}
  n8 = Saturate(None=Multiply(A=NamedRerouteUsage(), B=LinearInterpolate(A=@n5, B=@n6, Alpha=@n7)))
  n9 = ScalarParameter() {parameterName=IsBroken}
  Saturate(None=Multiply(A=LinearInterpolate(A=@n4, B=@n8, Alpha=@n9), B=ScalarParameter() {defaultValue=1, group=Alpha, parameterName=Final_Alpha}))
 ]
}
