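; WeaponsDynamicUIDirector is a Dynamic UI director that adds the weapons damage-numbers UI scene for the owning local player.
; On BeginPlay it gets the Fort global UI context from that player's controller and checks whether the context reports being in zone.
; Only when that check passes does it add WeaponsDynamicUIScene; the construction script is empty.
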(event EventBeginPlay
  (if (GlobalUIContext|IsinZone (PlayerController|LocalPlayerSubsystems|GetFortGlobalUIContext (DynamicUIDirector|GetOwningLocalPlayerController)))
    (DynamicUIDirector|AddScene "/DamageNumbers/WeaponsDynamicUIScene.WeaponsDynamicUIScene")))

(fn ConstructionScript)
